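; TNTina bow arrow projectile (Athena) that triggers an explosion when a fully charged shot hits.
; On the server, impact spawns a stationary dummy projectile slightly off the surface along the hit normal to avoid obstruction.
; That dummy carries the explode effect container spec and triggers the actual explosion.
; Fully charged impact FX adds a close explosion sound and a local gameplay cue for camera shake and controller vibration.
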(event EventOnFullyChargedImpact
  (params (Hit%20Result :type Hit%20Result%20Structure))
  ;[1 1 1] When fully charged projectile hits, tries to spawn an explosive projectile
  (if (Networking|IsServer)
    (CallFunction|SpawnExplosive Hit%20Result)))

(event Custom|SpawnExplosive
  (params (Hit :type Hit%20Result%20Structure))
  (bind (_bblockinghit _binitialoverlap _time _distance _location _impactpoint _normal) (Collision|BreakHitResult Hit))
  ;[1 1 1] Spawns a dummy explosive projectile with a small offset from the hit point to prevent weird obstructions. The new projectile is what triggers the explosion
  (Projectile|SpawnProjectile :ProjectileClass (Variables|Default|GetDummyProjectileClass) :RequestedBy (Game|GetInstigator) :SpawnLocation (+ _location (* _normal 8.0)) :SpawnRotation "0, 0, 0" :SpawnDirection (Variables|Default|GetZeroRotator) :EffectContainerSpecToApplyOnExplode (Variables|Delivery|GetEffectContainerSpecToApplyOnExplode) :bUseDefaultPhysics false :InitialSpeed 0.0 :GravityScale 0.0 :ChargePercent 0.0))

(event EventPlayFullyChargedImpactFX
  (params (HitResult :type Hit%20Result%20Structure))
  (bind (_bblockinghit _binitialoverlap _time _distance _location) (Collision|BreakHitResult HitResult))
  (|CallParentFunction HitResult)
  (Audio|PlaySoundatLocation (Variables|Default|GetCue_CloseSound) _location) ; Play Close Explosion
  (GameplayCue|ExecuteGameplayCue%28Burst%29Local self (Variables|Default|GetExplosionFeedbackCue) (Ability|GameplayCue|MakeGameplayCueParameters :NormalizedMagnitude 0.0 :RawMagnitude 0.0 :Location "0,0,0" :Normal "0,0,0" :Instigator (Actor|GetOwner) :EffectCauser (Actor|GetOwner) :GameplayEffectLevel 1 :AbilityLevel 1 :bReplicateLocationWhenUsingMinimalRepProxy false)) ; Cam shake + controller vibration
)

(fn ConstructionScript
  (|CallParentFunction))
