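; Explosive projectile blueprint for the TnTina bow's dummy grenade, which detonates as soon as it spawns.
; On BeginPlay it calls the parent, then waits one tick so the projectile finishes setting up; without the delay it dies without effect.
; On the authority side only, it then triggers DoExplosion and kills itself.
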(event EventBeginPlay
  ;[1 1 1] After being spawned, we delay the projectile for a tick so it has enough time to set up and then be destroyed. Without the tick, it just dies without doing anything.
  (|CallParentFunction)
  (Utilities|FlowControl|DelayUntilNextTick)
  (Networking|SwitchHasAuthority
    (:Authority
      (Delivery|DoExplosion)
      (Delivery|Kill))))

(fn ConstructionScript)
