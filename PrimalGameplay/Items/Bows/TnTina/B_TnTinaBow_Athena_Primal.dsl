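; Tiny Tina's bow weapon Blueprint (Athena Primal), a cluster bomb bow that blinks its lights when fully charged.
; The charged FX event starts a looping timer that alternately plays a timeline forward and backward, lerping the mesh's EmissiveMultiplier between off and on colors.
; Max-charge events cache local versus remote to pick 1P or 3P blink sounds, with local adding muzzle Niagara.
; Deactivation clears the timer and forces emissive off, as do equip and BeginPlay.
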(event FortWeapon|EventOnEquippedAndAttached
  (|CallParentFunction)
  (CallFunction|ForceEmissiveOff))

(event EventBeginPlay
  (|CallParentFunction)
  ;[1 1 1] Initialize the emissive strength to the "off" state
  (CallFunction|ForceEmissiveOff))

(event EventOnLocalClientReachedMaxCharge
  ;[1 1 1] Cache whether or not this is a local client or remote client so we can decide whether to play the 1P or 3P sound
  (|CallParentFunction)
  (Variables|Default|SetbIsLocalClient true)
  (Niagara|SpawnSystemAttached (Variables|ParentBow|GetNS_FullyCHarge) (Variables|Visual|GetWeaponMesh) "Muzzle"))

(event EventTick
  (params (DeltaSeconds :type Float%20%28single-precision%29))
  (|CallParentFunction DeltaSeconds) ; This node is disabled and will not be called.\nDrag off pins to build functionality.
)

(event EventOnRemoteClientReachedMaxCharge
  ;[1 1 1] Cache whether or not this is a local client or remote client so we can decide whether to play the 1P or 3P sound
  (|CallParentFunction)
  (Variables|Default|SetbIsLocalClient false))

(event EventSpawnFullyChargedFX
  ;[1 1 1] SpawnFullyChargedFX -> Starts the timer that drives the light blinking
  (Variables|Default|SetbLightsOn false)
  (CallFunction|ForceEmissiveOff)
  (bind _returnvalue (Utilities|Time|SetTimerbyFunctionName 0 "BlinkLights" (Variables|Default|GetLightBlinkFrequency) true))
  (Variables|Default|SetBlinkTimerHandle _returnvalue)
  (CallFunction|BlinkLights))

(event EventDeactivateFullyChargedFX
  ;[1 1 1] DeactivateFullyChargedFX -> Cancels blinking timer and makes sure the lights are off
  (|CallParentFunction)
  (Utilities|Time|ClearandInvalidateTimerbyHandle (Variables|Default|GetBlinkTimerHandle))
  (Variables|Default|SetbLightsOn false)
  (Components|Timeline|Stop (Variables|B_TnTinaBow_Athena_Primal|GetTL_LightBlink))
  (CallFunction|ForceEmissiveOff)
  (Development|PrintString "Cluster bomb bow FX deactivated"))

(event Custom|BlinkLights
  (bind _blightson (Variables|Default|GetbLightsOn))
  ;[1 1 1] BlinkLights -> Runs a timeline to lerp between the "on" and "off" emissive strength
  (Utilities|FlowControl|Sequence
    (:then_0
      (Components|Timeline|SetPlayRate (Variables|B_TnTinaBow_Athena_Primal|GetTL_LightBlink) (Utilities|Reciprocal (Variables|Default|GetBlinkEmissiveLerpTime)))
      (if (Variables|Default|GetbLightsOn)
        (bind (_direction _interpolation) (Timeline|TL_LightBlink PlayFromStart)
          (:Update
            (Rendering|Material|SetVectorParameterValueonMaterials (Variables|Visual|GetWeaponMesh) "EmissiveMultiplier" (Math|Conversions|ToVector(LinearColor) (Math|Color|Lerp(LinearColor) (Variables|Default|GetVectorParam_LightOff) (Variables|Default|GetVectorParam_LightOn) _interpolation)))))
        (else
          (Timeline|TL_LightBlink ReverseFromEnd))))
    (:then_1
      (if _blightson
        (CallFunction|PlayBlinkSound)
        (else
          _))
      (Variables|Default|SetbLightsOn (not _blightson)))))

(event Custom|PlayBlinkSound
  ;[1 1 1] Play blinking sound
  (Audio|PlaySoundatLocation (select (Variables|Default|GetbIsLocalClient) (Variables|Audio|GetBlinkSound_1P) (Variables|Audio|GetBlinkSound_3P)) (Transformation|GetActorLocation)))

(fn ForceEmissiveOff
  (Rendering|Material|SetVectorParameterValueonMaterials (Variables|Visual|GetWeaponMesh) "EmissiveMultiplier" (Math|Conversions|ToVector(LinearColor) (Variables|Default|GetVectorParam_LightOff))))

(fn ConstructionScript
  (|CallParentFunction))
